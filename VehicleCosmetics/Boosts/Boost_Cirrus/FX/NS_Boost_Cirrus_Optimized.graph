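# Optimized Niagara effect for the Cirrus car boost cosmetic, drawing glowing ribbon trails and a source mesh at the car's skeletal-mesh boost sockets.
# Two ribbon emitters spawn a trail per filtered port at quality-tiered rates, the narrower one swirled by a vortex force.
# A third emitter partitions rotating mesh and sprite particles across the ports.
# All three support painted color variants and branch on a frontend-preview flag to retune values like rate and lifetime.
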NiagaraSystem /VehicleCosmetics/Boosts/Boost_Cirrus/FX/NS_Boost_Cirrus_Optimized
dependencies {
 usedDataInterfaces [&NiagaraDataInterfaceCurve, &NiagaraDataInterfaceVectorCurve, &NiagaraDataInterfaceSkeletalMesh]
 usedDynamicInputs []
 usedModules [
  &self:SystemSpawnScript.SystemScriptSource.SystemScriptGraph.NiagaraNodeAssignment_0.SetVariables_#1
  &SystemState
 ]
}
emitters {
 NE_VC_Ribbon_CoreGLOW {
  emitterUpdateScript [
   &EmitterState(
    Inactive Response=::NewEnumerator1
    Life Cycle Mode=::NewEnumerator1
    Loop Duration=1
    Recalculate Duration Each Loop=true
    UseLoopDelay=false
   )
   &VC_FilteredSocketCount()
   &VC_RateQualityHandler(Epic Spawn Rate=30, High Spawn Rate=28, Low Spawn Rate=20, Medium Spawn Rate=25, Use Custom Rate?=true) {module=VC_RateQualityHandler001}
   &VC_RateQualityHandler()
   &VC_SpawnRibbon_v1(
    Number of Ribbons=$Emitter.PortCount
    SpawnRate=&MakeCustomFloatFromBool(
     Bool=$User.bIsFrontendPreview
     True Float=$Emitter.Spawn Rate
     False Float=&Multiply_Float(A=$Emitter.Spawn Rate, B=2)
    )
   )
  ]
  particleSpawnScript [
   &VC_PaintedVariantsOverride(
    Black=<rgba 0.05 0.05 0.05 1>
    BurntSienna=<rgba 0.35 0.0707 0.008 1>
    Cobalt=<rgba 0.0674 0.1516 1.75 1>
    Crimson=<rgba 0.7 0 0.019 1>
    DefaultColor=<rgba 0.1323 0.5 0.1568 1>
    ForestGreen=<rgba 0.0609 0.75 0.1184 1>
    Grey=<rgba 0.3 0.3 0.3 1>
    Lime=<rgba 0.56 1.2 0 1>
    Orange=<rgba 0.9604 0.335 0 1>
    Pink=<rgba 1 0.15 0.6909 1>
    Purple=<rgba 0.25 0 0.5 1>
    Saffron=<rgba 1 1 0 1>
    SkyBlue=<rgba 0.05 0.4458 1 1>
    TitaniumWhite=<rgba 1 1 1 2>
   )
   &VC_InitializeRibbon_v1(
    Color=$Particles.Color
    Enable Skeletal Position=true
    Lifetime=&MakeCustomFloatFromBool(Bool=$User.bIsFrontendPreview, True Float=1, False Float=0.5)
    Offset Amount:NiagaraFloat=&MakeFloatFromVector(
     Vector=&AddVectorToPosition(Position=<25 0 0>, Vector=&Multiply_VectorByFloat(Vector=$User.DTCarOffset, Float=1))
     Channel=::NewEnumerator0
    )
    Offset Axis:ENiagaraOrientationAxis=::XAxis
    Ribbon Width=25
   )
   &AddVelocity(
    Velocity=&MakeCustomVectorFromBool(Bool=$User.bIsFrontendPreview, True Vector=<2000 0 0>, False Vector=<2000 0 0>)
    Velocity Speed Scale=-1
   ) {module=AddVelocity001}
  ]
  particleUpdateScript [
   &ParticleState()
   &InheritSourceMovement(
    Applied Position Scale=&Multiply_VectorByFloat(
     Vector=<1 1 1>
     Float=&FloatFromCurve(FloatCurve=DataInterface() {propertyValues {Curve {keys [{interpMode=RCIM_Cubic, tangentMode=RCTM_User, time=0, value=1}, {interpMode=RCIM_Cubic, tangentMode=RCTM_User, time=1, value=0}]}}})
    )
    ApplyForce=false
    ApplyPosition=true
    ApplyPositionFalloff=false
   )
   &SolveForcesAndVelocity(Clamp Velocity=false, Limit Acceleration=false)
   &VC_UpdateRibbon_v1(Enable Skeletal Position=true)
   &DynamicMaterialParameters(
    Index 0 Param 1=&MakeCustomFloatFromBool(Bool=$User.bIsPainted, True Float=0.9, False Float=0.9)
    Index 0 Param 2=$User.SolidMask
    Index 0 Param 3=$User.AddMultiply
    Index 0 Param 4=$User.Brightness
    Module.Index0Param0Write=true
    Module.Index0Param1Write=true
    Module.Index0Param2Write=true
    Module.Index0Param3Write=true
   )
   &ScaleColor(
    Scale Alpha=&FloatFromCurve(
     FloatCurve=DataInterface() {propertyValues {Curve {keys [<at 0 1.5>,<at 0.5 1>,<at 1 0>]}}}
     Scale Curve=&MakeCustomFloatFromBool(Bool=$User.bIsFrontendPreview, True Float=1, False Float=1)
    )
    Scale RGB=&VectorFromCurve(
     VectorCurve=DataInterface() {propertyValues {XCurve {keys [<at 0 20>,<at 1 0>]}, YCurve {keys [<at 0 20>,<at 1 0>]}, ZCurve {keys [<at 0 20>,<at 1 0>]}}}
     Scale Curve=&Multiply_VectorByFloat(
      Vector=<1 1 1>
      Float=&MakeCustomFloatFromBool(Bool=$User.bIsFrontendPreview, True Float=1, False Float=0.75)
     )
    )
    ScaleA=true
    ScaleRGB=true
   )
   &ScaleRibbonWidth(Ribbon Width Scale=&FloatFromCurve(FloatCurve=DataInterface() {propertyValues {Curve {keys [
         {interpMode=RCIM_Cubic, tangentMode=RCTM_User, time=0.5, value=1}
         {arriveTangent=0.1606, interpMode=RCIM_Cubic, leaveTangent=0.1606, tangentMode=RCTM_User, time=1, value=0}
        ]}}}))
  ]
  rendererClasses [&NiagaraRibbonRendererProperties]
  renderers [{rendererClass=&NiagaraRibbonRendererProperties}]
 }
 NE_VC_Ribbon_CoreGLOW001 {
  emitterUpdateScript [
   &EmitterState(Life Cycle Mode=::NewEnumerator1, Loop Duration=0.5, UseLoopDelay=false)
   &VC_FilteredSocketCount()
   &VC_RateQualityHandler(Epic Spawn Rate=30, High Spawn Rate=28, Low Spawn Rate=20, Medium Spawn Rate=25, Use Custom Rate?=true) {module=VC_RateQualityHandler001}
   &VC_RateQualityHandler()
   &VC_SpawnRibbon_v1(
    Number of Ribbons=$Emitter.PortCount
    SpawnRate=&MakeCustomFloatFromBool(
     Bool=$User.bIsFrontendPreview
     True Float=$Emitter.Spawn Rate
     False Float=&Multiply_Float(A=$Emitter.Spawn Rate, B=2)
    )
   )
  ]
  particleSpawnScript [
   &VC_PaintedVariantsOverride(
    Black=<rgba 0.05 0.05 0.05 1>
    BurntSienna=<rgba 0.35 0.0707 0.008 1>
    Cobalt=<rgba 0.0674 0.1516 1.75 1>
    Crimson=<rgba 0.7 0 0.019 1>
    DefaultColor=<rgba 0.1323 0.5 0.1568 1>
    ForestGreen=<rgba 0.0609 0.75 0.1184 1>
    Grey=<rgba 0.3 0.3 0.3 1>
    Lime=<rgba 0.56 1.2 0 1>
    Orange=<rgba 0.9604 0.335 0 1>
    Pink=<rgba 1 0.15 0.6909 1>
    Purple=<rgba 0.25 0 0.5 1>
    Saffron=<rgba 1 1 0 1>
    SkyBlue=<rgba 0.05 0.4458 1 1>
    TitaniumWhite=<rgba 1 1 1 2>
   )
   &VC_InitializeRibbon_v1(
    Color=$Particles.Color
    Enable Skeletal Position=true
    Lifetime=&MakeCustomFloatFromBool(Bool=$User.bIsFrontendPreview, True Float=0.6, False Float=0.3)
    Offset Amount:NiagaraFloat=&MakeFloatFromVector(
     Vector=&AddVectorToPosition(Position=<25 0 0>, Vector=&Multiply_VectorByFloat(Vector=$User.DTCarOffset, Float=1))
     Channel=::NewEnumerator0
    )
    Offset Axis:ENiagaraOrientationAxis=::XAxis
    Ribbon Width=20
   )
   &AddVelocity(Velocity=<2500 0 75>, Velocity Speed Scale=-1) {module=AddVelocity001}
  ]
  particleUpdateScript [
   &ParticleState()
   &VortexForce(
    Origin Pull Only On Vortex Axis=false
    UseOriginOffset=false
    Vortex Axis=<1 0 0>
    Vortex Axis Coordinate Space:ENiagaraCoordinateSpace=::Local
    Vortex Origin:NiagaraPosition=$Particles.Initial.Position
   )
   &InheritSourceMovement(
    Applied Position Scale=&Multiply_VectorByFloat(
     Vector=<1 1 1>
     Float=&FloatFromCurve(FloatCurve=DataInterface() {propertyValues {Curve {keys [{interpMode=RCIM_Cubic, tangentMode=RCTM_User, time=0, value=1}, {interpMode=RCIM_Cubic, tangentMode=RCTM_User, time=1, value=0}]}}})
    )
    ApplyForce=false
    ApplyPosition=true
    ApplyPositionFalloff=false
   )
   &SolveForcesAndVelocity(Clamp Velocity=false, Limit Acceleration=false)
   &VC_UpdateRibbon_v1(Enable Skeletal Position=true)
   &DynamicMaterialParameters(
    Index 0 Param 1=&MakeCustomFloatFromBool(Bool=$User.bIsPainted, True Float=0.9, False Float=0.9)
    Index 0 Param 2=$User.SolidMask
    Index 0 Param 3=$User.AddMultiply
    Index 0 Param 4=$User.Brightness
    Module.Index0Param0Write=true
    Module.Index0Param1Write=true
    Module.Index0Param2Write=true
    Module.Index0Param3Write=true
   )
   &ScaleColor(
    Scale Alpha=&FloatFromCurve(
     FloatCurve=DataInterface() {propertyValues {Curve {keys [<at 0 1.5>,<at 0.5 1>]}}}
     Scale Curve=&MakeCustomFloatFromBool(Bool=$User.bIsFrontendPreview, True Float=1, False Float=1)
    )
    Scale RGB=&VectorFromCurve(
     VectorCurve=DataInterface() {propertyValues {XCurve {keys [<at 0 10>,<at 1 7.5>]}, YCurve {keys [<at 0 10>,<at 1 7.5>]}, ZCurve {keys [<at 0 10>,<at 1 7.5>]}}}
     Scale Curve=&Multiply_VectorByFloat(
      Vector=<1 1 1>
      Float=&MakeCustomFloatFromBool(Bool=$User.bIsFrontendPreview, True Float=1, False Float=0.75)
     )
    )
    ScaleA=true
    ScaleRGB=true
   )
   &ScaleRibbonWidth(Ribbon Width Scale=&FloatFromCurve(FloatCurve=DataInterface() {propertyValues {Curve {keys [
         {arriveTangent=-5.1158, interpMode=RCIM_Cubic, leaveTangent=-5.1158, tangentMode=RCTM_User, time=0.5, value=1}
         {arriveTangent=0.1606, interpMode=RCIM_Cubic, leaveTangent=0.1606, tangentMode=RCTM_User, time=1, value=0}
        ]}}}))
  ]
  rendererClasses [&NiagaraRibbonRendererProperties]
  renderers [
   {rendererClass=&NiagaraRibbonRendererProperties}
   {rendererClass=&NiagaraRibbonRendererProperties, rendererIndex=1}
  ]
 }
 NE_VC_SOURCE_MESH {
  emitterUpdateScript [
   &EmitterState(Life Cycle Mode=::NewEnumerator1, Loop Duration=1, UseLoopDelay=false)
   &VC_FilteredSocketCount()
   &SpawnRate(
    SpawnRate=&MakeCustomFloatFromBool(Bool=$User.bIsFrontendPreview, True Float=&RandomRangeFloat(Minimum=4, Maximum=8), False Float=30)
    Use Spawn Probability=false
   )
  ]
  particleSpawnScript [
   &VC_PaintedVariantsOverride(
    Black=<rgba 0.025 0.025 0.025 100>
    BurntSienna=<rgba 0.15 0.035 0 5>
    Cobalt=<rgba 0 0 1 5>
    Crimson=<rgba 0.5 0 0.0136 100>
    DefaultColor=<rgba 0.0625 0.2 0.05 10>
    ForestGreen=<rgba 0.0239 0.2 0.0079 2>
    Grey=<rgba 0.08 0.08 0.08 10>
    Lime=<rgba 0.2356 0.4 0.0054 5>
    Orange=<rgba 0.3 0.1 0 10>
    Pink=<rgba 0.5052 0.0605 0.3435 5>
    Purple=<rgba 0.1083 0 0.25 10>
    Saffron=<rgba 0.25 0.25 0.0151 5>
    SkyBlue=<rgba 0.0492 0.1661 0.4 4>
    TitaniumWhite=<rgba 0.3 0.3 0.3 7.5>
   )
   &InitializeParticle(
    Color=$Particles.Color
    Lifetime=&MakeCustomFloatFromBool(Bool=$User.bIsFrontendPreview, True Float=1, False Float=0.25)
    Mesh Scale=&MakeCustomVectorFromBool(
     Bool=$User.bIsFrontendPreview
     True Vector=&RandomRangeVector(Minimum=<0.5 1 1>, Maximum=<0.7 1 1>)
     False Vector=&RandomRangeVector(Minimum=<0.5 1 1>, Maximum=<0.5 1 1>)
    )
    Mesh Scale Mode=::NewEnumerator4
    Sprite Rotation Mode=::NewEnumerator2
    Sprite Size Mode=::NewEnumerator3
    UsePositionOffset=false
    Write Material Random=true
   )
   &PartitionParticles(Number Of Partitions=$Emitter.PortCount)
   &SkeletalMeshLocation(
    Bone / Socket Index=$Particles.PartitionParticles.Partition
    Bone / Socket Sampling Mode=::NewEnumerator5
    MaskSpawnGroup=false
    Mesh Sampling Type=::NewEnumerator6
    Sampled Position Offset=&MakeCustomVectorFromBool(
     Bool=$User.bIsFrontendPreview
     True Vector=&AddVectorToPosition(Position=<-120 0 0>, Vector=$User.DTCarOffset)
     False Vector=&AddVectorToPosition(Position=<-90 0 0>, Vector=&Multiply_VectorByFloat(Vector=$User.DTCarOffset, Float=1))
    )
    Skeletal Mesh=$System.SkeletalMesh
   )
   &AddVelocity() {enabled=false}
   &InitialMeshOrientation(
    Mesh Orientation Mode=::NewEnumerator2
    Orientation Coordinate Space=::Local
    Orientation Vector=<-1 0 0>
    Rotation=&RandomRangeVector(Minimum=<-1 0 0>, Maximum=<1 0 0>)
    Rotation Coordinate Space=::NewEnumerator0
    Use Rotation Vector=true
   )
  ]
  particleUpdateScript [
   &ParticleState()
   &UpdateMeshOrientation(Rotation Rate=&MakeCustomFloatFromBool(Bool=$User.bIsFrontendPreview, True Float=0.5, False Float=0.5))
   &SolveForcesAndVelocity(Clamp Velocity=false, Limit Acceleration=false)
   &DynamicMaterialParameters(
    Index 0 Param 1=&RandomRangeFloat(
     Minimum=&FloatFromCurve(
      FloatCurve=DataInterface() {propertyValues {Curve {keys [<at 0 0>,<at 1 -0.4>]}}}
      Scale Curve=&MakeCustomFloatFromBool(Bool=$User.bIsFrontendPreview, True Float=2, False Float=0.1)
     )
     Maximum=&FloatFromCurve(
      FloatCurve=DataInterface() {propertyValues {Curve {keys [<at 0 0>,<at 1 0>]}}}
      Scale Curve=&MakeCustomFloatFromBool(Bool=$User.bIsFrontendPreview, True Float=2, False Float=0.1)
     )
    )
    Index 0 Param 2=&MakeCustomFloatFromBool(Bool=$User.bIsFrontendPreview, True Float=2.9, False Float=2)
    Index 0 Param 3=&RandomRangeFloat(
     Minimum=0.3
     Maximum=&MakeCustomFloatFromBool(Bool=$User.bIsFrontendPreview, True Float=0.4, False Float=0.4)
    )
    Index 0 Param 4=&RandomRangeFloat(
     Minimum=0.5
     Maximum=&MakeCustomFloatFromBool(Bool=$User.bIsFrontendPreview, True Float=0.6, False Float=0.6)
     Evaluation Type=::NewEnumerator1
    )
    Module.Index0Param0Write=true
    Module.Index0Param1Write=true
    Module.Index0Param2Write=true
    Module.Index0Param3Write=true
   )
   &ScaleColor(
    Scale Alpha=&FloatFromCurve(FloatCurve=DataInterface() {propertyValues {Curve {keys [<at 0 1>,<at 1 0.5>]}}}, Scale Curve=2)
    Scale RGB=&Multiply_VectorByFloat(
     Vector=<1 1 1>
     Float=&MakeCustomFloatFromBool(
      Bool=$User.bIsFrontendPreview
      True Float=1.5
      False Float=&MakeCustomFloatFromBool(Bool=$User.bIsPainted, True Float=0.7, False Float=1.5)
     )
    )
    ScaleA=true
    ScaleRGB=true
   )
  ]
  rendererClasses [&NiagaraMeshRendererProperties, &NiagaraSpriteRendererProperties]
  renderers [
   {rendererClass=&NiagaraMeshRendererProperties, rendererIndex=0}
   {rendererClass=&NiagaraSpriteRendererProperties, rendererIndex=1}
  ]
 }
}
summary {
 systemName=NS_Boost_Cirrus_Optimized
 emitters [
  {
   emitterName=NE_VC_SOURCE_MESH
   rendererClasses [&NiagaraMeshRendererProperties, &NiagaraSpriteRendererProperties]
  }
  {emitterName=NE_VC_Ribbon_CoreGLOW, rendererClasses [&NiagaraRibbonRendererProperties]}
  {emitterName=NE_VC_Ribbon_CoreGLOW001, rendererClasses [&NiagaraRibbonRendererProperties]}
 ]
 userVariables [
  User.AddMultiply {type=NiagaraFloat, defaultValue=1, description=""}
  User.bIsFrontendPreview {type=NiagaraBool}
  User.bIsPainted {type=NiagaraBool}
  User.Brightness {type=NiagaraFloat, defaultValue=20, description=""}
  User.ColorVariantID {type=NiagaraInt32}
  User.DTCarOffset {type=Vector3f, description="", defaultValue {struct=Vector3f, value=<0 0 0>}}
  User.PaintedColor {type=LinearColor, description="", defaultValue {struct=LinearColor, value=<rgba 1 1 1 1>}}
  User.PaintedScalar {type=NiagaraFloat, defaultValue=1, description=""}
  User.PortCount {type=NiagaraInt32, defaultValue=2}
  User.SolidMask {type=NiagaraFloat, defaultValue=0.7, description=""}
 ]
}
