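# Unlit, two-sided masked material for the Cirrus boost cone effect, built for Niagara mesh particles and ribbons.
# Emissive layers a view-dependent gradient, edge glow and flames from panning, distorted noise, tinted by particle color.
# Repeating UV stripes, shaped by dynamic length and sharpness parameters, form the cutout mask with particle alpha and add highlights.
# Layered sine waves along vertex normals make a time-animated wobble, scaled by dynamic intensity and UV gradients.
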Material /VehicleCosmetics/Boosts/Boost_Cirrus/Materials/M_Boost_Cone_M_BoostAT
graph {
 properties {
  bUsedWithNiagaraMeshParticles=true
  bUsedWithNiagaraRibbons=true
  blendMode=BLEND_Masked
  pixelDepthOffsetMode=PDOM_Legacy
  shadingModel=MSM_Unlit
  twoSided=true
  usedShadingModels=Unlit
 }
 nodes [
  n1 = ParticleColor()
  n2 = TextureCoordinate() {uTiling=6.5}
  n3 = ComponentMask(None=Frac(None=Multiply(A=@n2))) {r=true}
  n4 = DynamicParameter() {paramNames [Lenght,Sharpness,Intensity,Param4]}
  n5 = Add(A=ComponentMask(None=TextureCoordinate()) {g=true}, B=@n4.Lenght)
  n6 = Multiply(A=Multiply(A=OneMinus(None=@n3), B=@n3), B=Clamp(None=Multiply(A=@n5, B=@n4.Sharpness))) {constB=3}
  n7 = ComponentMask(None=Panner(Coordinate=TextureCoordinate() {uTiling=2.8}, Time=Constant() {r=1}) {speedX=-1.5}) {r=true}
  n8 = Multiply(A=OneMinus(None=@n7), B=@n7)
  n9 = LinearInterpolate(A=Clamp(None=ConstantBiasScale(None=@n6) {bias=-0.1, scale=1.2}), Alpha=Clamp(None=Multiply(A=@n8) {constB=60})) {constB=0}
  n10 = Multiply(A=@n9) {constB=20}
  Clamp(None=Multiply(A=@n1.A, B=Clamp(None=@n10)))
  n11 = TextureCoordinate()
  n12 = Constant() {desc=.3, r=0.5}
  n13 = Multiply(A=Time(), B=ScalarParameter() {defaultValue=1, parameterName=Outer_Speed})
  n14 = Panner(Coordinate=TextureCoordinate() {vTiling=0.25}, Time=@n13) {speedX=-0.05, speedY=1}
  n15 = Panner(Coordinate=TextureCoordinate() {vTiling=0.5}, Time=@n13) {speedX=0.02, speedY=1}
  n16 = ComponentMask(None=Add(
    A=TextureSample(UVs=@n14) {samplerType=SAMPLERTYPE_Normal, texture=/DealershipSharedLibrary/FX/Textures/Noise/T_VCVFX_Water02_N}.RGB
    B=TextureSample(UVs=@n15) {samplerType=SAMPLERTYPE_Normal, texture=&T_VCVFX_Water02_N}.RGB
   )) {g=true, r=true}
  n17 = ComponentMask(None=Add(A=@n11, B=Multiply(A=@n12, B=@n16))) {g=true}
  n18 = Power(Base=@n17, Exp=ScalarParameter() {defaultValue=8, parameterName=GradientAmount})
  n19 = ScalarParameter() {defaultValue=2, parameterName=GradientSharpness}
  n20 = DotProduct(
   A=Transform(None=CameraVectorWS()) {transformSourceType=TRANSFORMSOURCE_Camera, transformType=TRANSFORM_View}
   B=Constant3Vector() {constant=<rgba 1 0 0 0>}
  )
  n21 = Power(Base=ComponentMask(None=@n11) {g=true}) {constExponent=4}
  n22 = LinearInterpolate(
   A=ScalarParameter() {defaultValue=6, parameterName=FresnelEnd}
   B=ScalarParameter() {defaultValue=3, parameterName=FresnelBase}
   Alpha=Clamp(None=Multiply(A=@n21) {constB=4})
  )
  n23 = Multiply(A=Clamp(None=Multiply(A=@n18, B=@n19)), B=Clamp(None=Power(Base=@n20, Exp=@n22)))
  n24 = Clamp(None=@n23)
  n25 = ScalarParameter() {defaultValue=10, parameterName=EdgeThickness}
  n26 = Clamp(None=OneMinus(None=Multiply(A=@n25, B=@n23)))
  n27 = Clamp(None=Multiply(A=ScalarParameter() {defaultValue=30, parameterName=EdgeGlow}, B=Multiply(A=@n24, B=@n26)))
  n28 = ScalarParameter() {defaultValue=1, parameterName=EdgeGlowBrightness}
  n29 = Multiply(B=@n24) {constA=0.9}
  n30 = Constant() {r=0.5}
  n31 = Constant() {r=0.5}
  n32 = Panner(Coordinate=TextureCoordinate() {uTiling=0.25}, Time=@n31) {speedX=0.5}
  n33 = TextureSample(UVs=Panner(Coordinate=TextureCoordinate()) {speedX=0.01, speedY=0.125}) {texture=/DealershipSharedLibrary/FX/Textures/Noise/T_VCVFX_Flow02_N}
  n34 = Multiply(A=Constant() {desc=.3, r=0.05}, B=AppendVector(A=@n33.R, B=@n33.R))
  n35 = TextureSampleParameter2D(UVs=Panner(Coordinate=Add(A=@n32, B=@n34)) {speedY=1}) {parameterName=TrailMask_RGB_T, texture=/DealershipSharedLibrary/FX/Textures/Noise/T_VCVFX_TrailMask_RGB}
  n36 = Panner(Coordinate=TextureCoordinate() {uTiling=-0.25, vTiling=1.25}, Time=@n31) {speedX=-0.5}
  n37 = TextureSampleParameter2D(UVs=Panner(Coordinate=Add(A=@n36, B=@n34)) {speedY=1}) {parameterName=TrailMask_RGB_T, texture=&T_VCVFX_TrailMask_RGB}
  n38 = Multiply(A=Constant() {desc=.3, r=0.05}, B=ComponentMask(None=Add(A=@n35.RGB, B=@n37.RGB)) {g=true, r=true})
  n39 = Add(A=Panner(Coordinate=TextureCoordinate() {uTiling=2, vTiling=1.25}, Time=@n30) {speedX=0.5}, B=@n38)
  n40 = Multiply(A=Time()) {constB=5}
  n41 = ChannelMaskParameter(None=TextureSample(UVs=Panner(Coordinate=@n39, Time=@n40) {speedX=0.25, speedY=1}) {texture=/DealershipSharedLibrary/FX/Textures/Noise/T_VCVFX_NoiseCloud_RGB}.RGB) {parameterName=MaskRGB_01}
  n42 = Add(A=@n38, B=Panner(Coordinate=TextureCoordinate() {uTiling=-2}, Time=@n30) {speedX=-0.5})
  n43 = ChannelMaskParameter(None=TextureSample(UVs=Panner(Coordinate=@n42, Time=@n40) {speedX=0.2, speedY=0.5}) {texture=&T_VCVFX_NoiseCloud_RGB}.RGB) {parameterName=MaskRGB_02}
  n44 = Multiply(A=@n41.R, B=@n43.R)
  n45 = Add(A=@n41.R, B=@n43.R)
  n46 = Multiply(
   A=ScalarParameter() {defaultValue=5, parameterName=Flames}
   B=Multiply(A=@n23, B=StaticSwitchParameter(True=@n44, False=@n45) {parameterName=MultiplyORadd})
  )
  n47 = Add(A=Add(A=Multiply(A=@n27, B=@n28), B=Add(A=@n29, B=@n46)), B=Clamp(None=ConstantBiasScale(None=@n17) {bias=-0.8, scale=1}))
  n48 = Clamp(None=Subtract(A=OneMinus(None=@n9), B=Multiply(A=@n9) {constB=3}))
  n49 = Multiply(B=@n48) {constA=5}
  MaterialFunctionCall(Emissive=Multiply(A=LinearInterpolate(A=@n47, B=@n49, Alpha=@n48), B=@n1.RGB)) {materialFunction=&MF_FX_EmissiveAdjustments}
  n50 = ComponentMask(None=TextureCoordinate()) {g=true}
  n51 = Add(A=OneMinus(None=@n50)) {constB=1.8}
  n52 = Transform(None=VertexNormalWS()) {transformSourceType=TRANSFORMSOURCE_World}
  n53 = Multiply(B=Time()) {constA=6}
  n54 = Add(A=Multiply(B=Time()) {constA=2}, B=ComponentMask(None=TextureCoordinate() {uTiling=0.5}) {r=true})
  n55 = Add(A=Sine(None=Add(A=@n53, B=@n50)), B=Multiply(B=Sine(None=@n54)) {constA=0.7})
  n56 = Multiply(A=@n55) {constB=0.75}
  n57 = ComponentMask(None=TextureCoordinate() {uTiling=4, vTiling=4}) {r=true}
  n58 = Multiply(A=Time()) {constB=0.1}
  n59 = Sine(None=Add(A=Add(A=@n56, B=@n57), B=Multiply(B=@n58) {constA=5.24}))
  n60 = Multiply(A=Add(A=@n55, B=Multiply(A=@n59, B=@n4.Param4)), B=@n4.Intensity)
  n61 = TextureCoordinate()
  n62 = Multiply(A=OneMinus(None=ComponentMask(None=@n61) {g=true})) {constB=7}
  n63 = Multiply(A=Time()) {constB=0.1}
  n64 = Sine(None=@n63) {period=1.5}
  n65 = Cosine(None=@n63) {period=2}
  n66 = Multiply(A=Time()) {constB=0.1}
  n67 = Sine(None=@n66)
  n68 = Cosine(None=@n66)
  n69 = LinearInterpolate(A=Multiply(A=Clamp(None=@n64)) {constB=0.5}, B=Multiply(A=Clamp(None=@n65)) {constB=0.5}, Alpha=Clamp(None=Add(A=@n67, B=@n68)))
  Multiply(
   A=Multiply(A=Multiply(B=@n51) {constA=2}, B=Multiply(A=@n52, B=@n60))
   B=Multiply(A=Clamp(None=@n62) {maxDefault=3}, B=Add(A=@n69) {constB=2})
  )
  Multiply() {constB=0.1}
 ]
}
