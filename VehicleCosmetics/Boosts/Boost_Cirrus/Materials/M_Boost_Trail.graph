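# Translucent, two-sided trail material for the Cirrus vehicle boost cosmetic, built for Niagara ribbons and tinted by particle color.
# Two panning cloud-noise layers combine with a flow-distorted band mask along V, faded at the ribbon ends, to form emissive and opacity.
# Dynamic parameters Hardness, SolidMask, AddMultiply and Brightness control the mask shape and blend.
# Static switches add optional distortion and a layered sine/cosine world position offset wobble that ramps in along U.
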Material /VehicleCosmetics/Boosts/Boost_Cirrus/Materials/M_Boost_Trail
graph {
 properties {
  bUsedWithNiagaraRibbons=true
  blendMode=BLEND_Translucent
  pixelDepthOffsetMode=PDOM_Legacy
  refractionMethod=RM_2DOffset
  twoSided=true
  usedShadingModels=Default Lit
 }
 nodes [
  n1 = ParticleColor()
  Desaturation(None=@n1.RGB, Fraction=Constant() {r=0.1})
  Desaturation(None=@n1.RGB)
  n2 = TextureCoordinate() {uTiling=1.5, vTiling=0.5}
  n3 = Constant() {r=0.5}
  n4 = Multiply(A=Time(), B=ScalarParameter() {defaultValue=2, parameterName=PanSpeed})
  n5 = TextureSample(UVs=Panner(Coordinate=Panner(Coordinate=@n2, Time=@n3) {speedX=0.25}, Time=@n4) {speedX=-1.5, speedY=-0.2}) {texture=/DealershipSharedLibrary/FX/Textures/Noise/T_VCVFX_NoiseCloud_RGB}
  n6 = TextureCoordinate() {vTiling=-0.5}
  n7 = TextureSample(UVs=Panner(Coordinate=Panner(Coordinate=@n6, Time=@n3) {speedY=-0.75}, Time=@n4) {speedX=-1, speedY=-0.2}) {texture=&T_VCVFX_NoiseCloud_RGB}
  n8 = Desaturation(None=Add(A=@n5.RGB, B=@n7.RGB))
  n9 = Clamp(None=Multiply(B=@n8) {constA=0.15})
  n10 = Desaturation(None=Multiply(A=@n5.RGB, B=@n7.RGB))
  n11 = StaticComponentMaskParameter(None=@n5.RGB) {defaultR=true, parameterName=MaskRGB_01}
  n12 = StaticComponentMaskParameter(None=@n7.RGB) {defaultR=true, parameterName=MaskRGB_02}
  n13 = DynamicParameter() {paramNames [Hardness,SolidMask,AddMultiply,Brightness]}
  n14 = Clamp(None=Multiply(A=Multiply(A=@n11, B=@n12), B=@n13.Brightness))
  n15 = LinearInterpolate(A=@n14, Alpha=@n13.AddMultiply)
  n16 = Add(A=Multiply(A=Add(A=@n8, B=@n10)) {constB=0.1}, B=@n15)
  n17 = Add(A=@n11, B=@n12)
  n18 = Add(A=@n14, B=Clamp(None=Multiply(A=@n13.Brightness, B=@n17)))
  n19 = Clamp(None=@n18)
  n20 = Constant() {r=0.5}
  n21 = TextureSample(UVs=Panner(Coordinate=TextureCoordinate() {uTiling=0.25, vTiling=0.25})) {texture=/DealershipSharedLibrary/FX/Textures/Noise/T_VCVFX_Flow02_N}
  n22 = Multiply(A=Constant() {desc=.3, r=0.05}, B=AppendVector(A=@n21.R, B=@n21.R)) {constA=0.05}
  n23 = Panner(Coordinate=Add(A=Panner(Time=@n20) {speedY=0.5}, B=@n22), Time=@n4) {speedX=-1}
  n24 = Panner(Coordinate=Add(A=Panner(Time=@n20) {speedY=-0.5}, B=@n22), Time=@n4) {speedX=-1}
  n25 = ComponentMask(None=Add(
    A=TextureSample(UVs=@n23) {texture=/DealershipSharedLibrary/FX/Textures/Noise/T_VCVFX_DistortNoise02_RGB}.RGB
    B=TextureSample(UVs=@n24) {texture=&T_VCVFX_DistortNoise02_RGB}.RGB
   )) {g=true, r=true}
  n26 = ComponentMask(None=Add(A=TextureCoordinate(), B=Multiply(B=@n25) {constA=0.5})) {g=true}
  n27 = OneMinus(None=@n26)
  n28 = Multiply(A=Clamp(None=ConstantBiasScale(None=@n26) {bias=-0.15, scale=1}), B=Clamp(None=ConstantBiasScale(None=@n27) {bias=-0.15, scale=1}))
  n29 = Multiply(A=Clamp(None=Multiply(A=@n13.Hardness, B=@n28)), B=@n13.Brightness)
  n30 = Clamp(None=Multiply(A=@n28) {constB=6})
  n31 = Clamp(None=Multiply(A=LinearInterpolate(A=@n29, B=@n30, Alpha=@n13.SolidMask), B=@n15))
  n32 = LinearInterpolate(A=@n29, B=Clamp(None=Multiply(A=@n31, B=@n13.Hardness)), Alpha=@n13.SolidMask)
  n33 = Multiply(A=Clamp(None=Subtract(A=@n19, B=@n32)))
  MaterialFunctionCall(Emissive=LinearInterpolate(A=Multiply(A=@n9, B=@n1.RGB), B=Multiply(A=@n1.RGB, B=@n16), Alpha=Clamp(None=@n33))) {materialFunction=&MF_FX_EmissiveAdjustments}
  Multiply(A=@n18, B=ParticleColor().RGBA)
  n34 = ComponentMask(None=TextureCoordinate()) {r=true}
  n35 = ComponentMask(None=TextureCoordinate()) {r=true}
  n36 = Clamp(None=ConstantBiasScale(None=@n35) {bias=-0.01, scale=7})
  n37 = Multiply(A=@n32, B=@n36)
  Multiply(A=Multiply(A=Multiply(A=@n34) {constB=8}, B=OneMinus(None=@n34)), B=Clamp(None=Multiply(A=@n1.RGBA, B=@n37)) {maxDefault=1.25})
  n38 = Multiply(A=@n30, B=Multiply(A=@n8, B=@n10))
  n39 = Clamp(None=OneMinus(None=@n35))
  StaticSwitchParameter(True=Multiply(A=Multiply(A=@n38) {constB=5}, B=Multiply(A=@n36, B=@n39)), False=Constant()) {parameterName=Distortion}
  n40 = Multiply(A=ScalarParameter() {parameterName=Speed}, B=OneMinus(None=Time()))
  n41 = ComponentMask(None=Multiply(A=TextureCoordinate() {uTiling=3}, B=ScalarParameter() {parameterName=Tile})) {r=true}
  n42 = Add(A=@n40, B=@n41)
  n43 = Multiply(A=@n40) {constB=0.5}
  n44 = Add(A=Add(A=Cosine(None=@n42), B=Cosine(None=@n42) {period=2}), B=Cosine(None=Add(A=@n41, B=@n43)) {period=0.75})
  n45 = Add(A=@n41, B=@n40)
  n46 = Sine(None=@n45)
  n47 = Sine(None=@n45) {period=2}
  n48 = AppendVector(A=@n44, B=Add(A=Add(A=@n46, B=@n47), B=Sine(None=@n45) {period=0.75}))
  n49 = Multiply(
   A=Multiply(A=AppendVector(A=@n48, B=@n44), B=ScalarParameter() {parameterName=Amount})
   B=ScalarParameter() {defaultValue=1, parameterName=Intensity}
  )
  StaticSwitchParameter(True=Multiply(A=Clamp(None=@n36), B=Transform(None=@n49)), False=Constant()) {parameterName=WPO}
 ]
}
